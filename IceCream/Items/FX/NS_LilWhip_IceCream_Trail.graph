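# Niagara ribbon trail effect for the LilWhip ice cream item.
# A single Ribbon emitter spawns particles at a steady rate with short lifetimes, and a cubic curve narrows ribbon width over each particle's life.
# The system also references a spawn-time variable assignment, a color-array selection input and an integer User.ColorInt parameter, suggesting selectable trail colors.
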NiagaraSystem /IceCream/Items/FX/NS_LilWhip_IceCream_Trail
dependencies {
 usedDataInterfaces [&NiagaraDataInterfaceCurve, &NiagaraDataInterfaceArrayColor]
 usedDynamicInputs [&SelectLinearColorFromArray]
 usedModules [
  &SystemState
  &self:SystemSpawnScript.SystemScriptSource.SystemScriptGraph.NiagaraNodeAssignment_0.SetVariables_#1
 ]
}
emitters {Ribbon {
  emitterUpdateScript [
   &EmitterState()
   &SpawnRate(SpawnRate=100, Use Spawn Probability=false)
  ]
  particleSpawnScript [&InitializeRibbon(Lifetime=0.25, Ribbon Width=90, UsePositionOffset=false, Write Material Random=true)]
  particleUpdateScript [
   &ParticleState()
   &ScaleRibbonWidth(Ribbon Width Scale=&FloatFromCurve(FloatCurve=DataInterface() {propertyValues {Curve {keys [
         {interpMode=RCIM_Cubic, leaveTangentWeight=0.9207, tangentMode=RCTM_Break, time=0, value=1}
         {arriveTangentWeight=1.3126, interpMode=RCIM_Cubic, tangentMode=RCTM_Break, time=1, value=0.2}
        ]}}}))
  ]
  rendererClasses [&NiagaraRibbonRendererProperties]
  renderers [{rendererClass=&NiagaraRibbonRendererProperties}]
 }}
summary {
 systemName=NS_LilWhip_IceCream_Trail
 emitters [{emitterName=Ribbon, rendererClasses [&NiagaraRibbonRendererProperties]}]
 userVariables [User.ColorInt {type=NiagaraInt32}]
}
