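; Projectile Blueprint for the thrown ice cream consumable, adding impact effects on top of its parent class's behavior.
; The spawned-item FX and bounce events each call the parent, spawn the FXImpact Niagara system at the actor's location and store it in FXImpactReference.
; The water impact event calls the parent, then spawns a fixed medium water splash instead.
; Pickup spawning and construction only call the parent; BeginPlay, overlap and Tick do nothing here.
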(event EventBeginPlay
  (|CallParentFunction) ; This node is disabled and will not be called.\nDrag off pins to build functionality.
)

(event Collision|EventActorBeginOverlap
  (params (OtherActor :type Actor%20Object%20Reference)))

(event EventTick
  (params (DeltaSeconds :type Float%20%28single-precision%29)))

(event EventOnPickupSpawned
  (params (Spawned%20Pickup :type Fort%20Pickup%20Object%20Reference))
  (|CallParentFunction Spawned%20Pickup))

(event EventSpawnedItemFx
  (|CallParentFunction)
  (bind _returnvalue (Niagara|SpawnSystematLocation (Variables|Default|GetFXImpact) (Transformation|GetActorLocation)))
  (Variables|Default|SetFXImpactReference _returnvalue))

(event EventOnBounce
  (params (Hit :type Hit%20Result%20Structure%20%28by%20ref%29))
  (|CallParentFunction Hit)
  (bind _returnvalue (Niagara|SpawnSystematLocation (Variables|Default|GetFXImpact) (Transformation|GetActorLocation)))
  (Variables|Default|SetFXImpactReference _returnvalue))

(event EventWaterImpactItemFX
  (|CallParentFunction)
  (Niagara|SpawnSystematLocation "/Game/Effects/Niagara/Systems/Enviroment/Water/NS_WaterSplash_Medium.NS_WaterSplash_Medium" (Transformation|GetActorLocation)))

(fn ConstructionScript
  (|CallParentFunction))
