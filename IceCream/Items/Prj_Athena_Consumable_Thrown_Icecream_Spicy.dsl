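; Projectile Blueprint for the Spicy variant of the thrown Icecream consumable, which only customizes the impact effect's color.
; The spawned-item FX and bounce events run parent logic, then set the impact Niagara effect's User.ColorInt to this variant's index.
; All other behavior comes from the parent, as the BeginPlay, overlap and Tick events are disabled stubs and the construction script only calls the parent.
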(event EventBeginPlay
  (|CallParentFunction) ; This node is disabled and will not be called.\nDrag off pins to build functionality.
)

(event Collision|EventActorBeginOverlap
  (params (OtherActor :type Actor%20Object%20Reference))
  (|CallParentFunction OtherActor) ; This node is disabled and will not be called.\nDrag off pins to build functionality.
)

(event EventTick
  (params (DeltaSeconds :type Float%20%28single-precision%29))
  (|CallParentFunction DeltaSeconds) ; This node is disabled and will not be called.\nDrag off pins to build functionality.
)

(event EventSpawnedItemFx
  (|CallParentFunction)
  (Niagara|SetNiagaraVariable(Int32) (Variables|Default|GetFXImpactReference) "User.ColorInt" 4))

(event EventOnBounce
  (params (Hit :type Hit%20Result%20Structure%20%28by%20ref%29))
  (|CallParentFunction Hit)
  (Niagara|SetNiagaraVariable(Int32) (Variables|Default|GetFXImpactReference) "User.ColorInt" 4))

(fn ConstructionScript
  (|CallParentFunction))
